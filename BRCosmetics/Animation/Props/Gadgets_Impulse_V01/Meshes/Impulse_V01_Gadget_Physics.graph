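# Physics asset for the Impulse V01 gadget prop's skeletal mesh, under the BRCosmetics animation props.
# It defines a single capsule body on the JNT_C_Sweater_Bind bone, offset and rotated relative to that bone.
# It has no constraints, so the one capsule is a coarse physics shape for the mesh with no articulation between parts.
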PhysicsAsset /BRCosmetics/Animation/Props/Gadgets_Impulse_V01/Meshes/Impulse_V01_Gadget_Physics
bodies {JNT_C_Sweater_Bind [{shapeType=Capsule, length=25.6266, radius=10.8324, center=<0.8858 0.2815 -8.3774>, rotation=<rot 0 -89.827 -89.5321>}]}
constraints []
